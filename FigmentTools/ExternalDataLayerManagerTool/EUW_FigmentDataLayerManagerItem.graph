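# Editor utility list-entry widget for the Figment Data Layer Manager, providing a button that switches the editor to one release state.
# When bound to an ExternalDataLayerReleaseItem, it caches that item's plugin, layer and label data.
# Clicking asynchronously changes plugin states, activating the release's plugins or the whole pool if flagged keep-all-loaded.
# On completion, it unloads and loads external, then in-level data layers via DataLayerEditorSubsystem, then broadcasts the state name.
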EditorUtilityWidgetBlueprint /FigmentTools/ExternalDataLayerManagerTool/EUW_FigmentDataLayerManagerItem
graphs [
 EventGraph {nodes [
   "On Clicked (ReleaseButton)"() {is=title K2Node_ComponentBoundEvent}
   > n1 = SetupPluginsForActivation!(execute, self, Plugins in Release:Array of Strings=Get Plugins in Release() -> (Plugins in Release:Array of Strings)) -> (then, PluginsToActivate:Array of Strings, TotalPlugins:Array of Strings)
   > n2 = ChangePluginsState!(execute:Exec, PluginsToActivate:Array of Strings=@n1, PluginsPool:Array of Strings=@n1.TotalPlugins) -> (then:Exec, OnStateChangeComplete:Exec) {is=title K2Node_AsyncAction}
   HandleDataLayerListInLevel!(
    execute=@n2.OnStateChangeComplete
    self
    DataLayerNamesToActivate:Array of Names=Get External Data Layers To Activate() -> (External Data Layers To Activate:Array of Names)
    DataLayerNamesToUnload:Array of Names=Get External Data Layers To Unload() -> (External Data Layers To Unload:Array of Names)
   ) -> (then)
   > HandleDataLayerListInLevel(
    DataLayerNamesToActivate=Get Data Layers in Level to Activate() -> (Data Layers in Level to Activate:Array of Names)
    DataLayerNamesToUnload=Get Data Layers in Level to Unload() -> (Data Layers in Level to Unload:Array of Names)
   )
   > NotifyStateNameChanged!(
    execute
    self:External Data Layer Release Item Object Reference=Get EDL_ReleaseItem() -> (EDL_ReleaseItem:External Data Layer Release Item Object Reference)
    CurretStateName:"Text (by ref)"=Get StateName() -> (StateName:Text)
   ) -> (then)
   n3 = Event OnListItemObjectSet()
   > n4 = Cast To ExternalDataLayerReleaseItem!(execute, Object=@n3.ListItemObject) -> (then, CastFailed, AsExternal Data Layer Release Item:External Data Layer Release Item Object Reference) {is=title K2Node_DynamicCast}
   > Set EDL_ReleaseItem(EDL_ReleaseItem:External Data Layer Release Item Object Reference=@n4) -> (Output_Get:External Data Layer Release Item Object Reference)
   > Initialize()
   Event PreConstruct()
   Event Construct()
   Event Tick()
  ]}
 HandleDataLayerListInLevel {nodes [
   n1 = HandleDataLayerListInLevel!() -> (then, DataLayerNamesToActivate:Array of Names, DataLayerNamesToUnload:Array of Names) {is=title K2Node_FunctionEntry}
   > Clear(TargetArray:Array of Data Layer Instance Object References=Get DataLayersToUnload() -> (DataLayersToUnload:Array of Data Layer Instance Object References))
   > Clear(TargetArray=Get DataLayersToActivate() -> (DataLayersToActivate:Array of Data Layer Instance Object References))
   > n2 = For Each Loop(Array:Array of Names=@n1) -> (Array Element:Name)
   n3 = For Each Loop(Exec=@n2.Completed, Array=@n1.DataLayerNamesToUnload)
   n4 = Get DataLayerEditorSubsystem() -> (ReturnValue:Data Layer Editor Subsystem Object Reference) {is=title K2Node_GetEditorSubsystem}
   SetDataLayersIsLoadedInEditor!(
    execute=@n3.Completed
    self:Data Layer Editor Subsystem Object Reference=@n4
    DataLayers:Array of Data Layer Instance Object References=Get DataLayersToUnload()
    bIsLoadedInEditor:bool=false
    bIsFromUserChange:bool=true
   ) -> (then, ReturnValue:bool)
   > SetDataLayersIsLoadedInEditor(self=Get DataLayerEditorSubsystem(), DataLayers=Get DataLayersToActivate(), bIsLoadedInEditor=true, bIsFromUserChange=true)
   n5 = GetDataLayerFromLabel!(execute, self:Data Layer Editor Subsystem Object Reference=@n4, DataLayerLabel:"Name (by ref)"=@n3) -> (then, ReturnValue:Data Layer Instance Object Reference)
   Is Valid(exec=@n3.LoopBody, InputObject=@n5)
   > Add(
    TargetArray:Array of Data Layer Instance Object References=Get DataLayersToUnload()
    NewItem:"Data Layer Instance Object Reference (by ref)"=@n5
   ) {is=title K2Node_CallArrayFunction}
   n6 = GetDataLayerFromLabel(self=Get DataLayerEditorSubsystem(), DataLayerLabel=@n2)
   Is Valid(exec=@n2.LoopBody, InputObject=@n6)
   > Add(TargetArray=Get DataLayersToActivate(), NewItem=@n6)
  ]}
 Initialize {nodes [
   Initialize() {is=title K2Node_FunctionEntry}
   n1 = Get EDL_ReleaseItem()
   n2 = Get!(self:External Data Layer Release Item Object Reference=@n1) -> (
    EDL_CommonGroupItemData_bKeepAllPluginsLoaded:bool=~
    EDL_CommonGroupItemData_PluginsPool:Array of Strings
    EDL_CommonGroupItemData_PluginsInRelease:Array of Strings
    EDL_CommonGroupItemData_ExternalDataLayersToActivate:Array of Names
    EDL_CommonGroupItemData_ExternalDataLayersToUnload:Array of Names
   )
   > Set KeepAllPluginsLoaded(KeepAllPluginsLoaded:bool=@n2) -> (Output_Get:bool)
   > Set Plugins Pool(Plugins Pool:Array of Strings=@n2.EDL_CommonGroupItemData_PluginsPool) -> (Output_Get:Array of Strings)
   > Set Plugins in Release(Plugins in Release:Array of Strings=@n2.EDL_CommonGroupItemData_PluginsInRelease) -> (Output_Get:Array of Strings)
   > Set External Data Layers To Activate(External Data Layers To Activate:Array of Names=@n2.EDL_CommonGroupItemData_ExternalDataLayersToActivate) -> (Output_Get:Array of Names)
   > Set External Data Layers To Unload(External Data Layers To Unload:Array of Names=@n2.EDL_CommonGroupItemData_ExternalDataLayersToUnload) -> (Output_Get:Array of Names)
   n3 = Get!(self=@n1) -> (
    EDL_ReleaseButtonActionData_DataLayersInLevelToActivate:Array of Names
    EDL_ReleaseButtonActionData_DataLayersInLevelToUnload:Array of Names
    EDL_ReleaseButtonActionData_StateName:Text
    EDL_ReleaseButtonActionData_ButtonText:Text
   )
   > Set Data Layers in Level to Activate(Data Layers in Level to Activate:Array of Names=@n3) -> (Output_Get:Array of Names)
   > Set Data Layers in Level to Unload(Data Layers in Level to Unload:Array of Names=@n3.EDL_ReleaseButtonActionData_DataLayersInLevelToUnload) -> (Output_Get:Array of Names)
   > Set StateName(StateName:Text=@n3.EDL_ReleaseButtonActionData_StateName) -> (Output_Get:Text)
   > Set Text(Text=@n3.EDL_ReleaseButtonActionData_ButtonText, self=Get ButtonTextBlock() -> (ButtonTextBlock:Text Object Reference))
  ]}
 OnDataLayersHandled {nodes [OnDataLayersHandled() {is=title K2Node_FunctionEntry}]}
 SetupPluginsForActivation {nodes [
   n1 = SetupPluginsForActivation!() -> (then, Plugins in Release:Array of Strings) {is=title K2Node_FunctionEntry}
   > UpdateKeepAllPluginsLoaded()
   > n2 = Branch(Condition=Get KeepAllPluginsLoaded() -> (KeepAllPluginsLoaded:bool))
   n3 = Get Plugins Pool() -> (Plugins Pool:Array of Strings)
   > Return Node!(execute, PluginsToActivate:Array of Strings=@n3, TotalPlugins:Array of Strings=@n3)
   Return Node(execute=@n2.else, PluginsToActivate=@n1, TotalPlugins=Get Plugins Pool())
  ]}
 UpdateKeepAllPluginsLoaded {nodes [
   UpdateKeepAllPluginsLoaded() {is=title K2Node_FunctionEntry}
   > Set KeepAllPluginsLoaded(KeepAllPluginsLoaded=Get!(self=Get EDL_ReleaseItem()) -> (
     EDL_CommonGroupItemData_bKeepAllPluginsLoaded:bool=~
     EDL_CommonGroupItemData_PluginsPool:Array of Strings
     EDL_CommonGroupItemData_PluginsInRelease:Array of Strings
     EDL_CommonGroupItemData_ExternalDataLayersToActivate:Array of Names
     EDL_CommonGroupItemData_ExternalDataLayersToUnload:Array of Names
    ))
  ]}
]
variables [
 Data Layers in Level to Activate {type="array<string>", default []}
 Data Layers in Level to Unload {type="array<string>", default []}
 EDL_ReleaseItem {type=ExternalDataLayerReleaseItem, default=null}
 External Data Layers To Activate {type="array<string>", default []}
 External Data Layers To Unload {type="array<string>", default []}
 KeepAllPluginsLoaded {type=boolean}
 OnDataLayersHandled {default=()}
 Plugins in Release {type="array<string>", default []}
 Plugins Pool {type="array<string>", default []}
 StateName {type=string, default=""}
]
widgets {
 info {rootWidgetClass=&VerticalBox, widgetCount=3}
 widgets [
  VerticalBox_0 {widgetClassPath=&VerticalBox, widget=VerticalBox_0}
  ReleaseButton {
   widgetClassPath=&EditorUtilityButton
   parent=VerticalBox_0
   widget=ReleaseButton
   properties {toolTipText=This will load and unload the proper layers based on the assigned row.}
   slot {class=VerticalBoxSlot, properties {padding=<ltrb 50 5 50 5>}}
  }
  ButtonTextBlock {
   widgetClassPath=&TextBlock
   bIsVariable=true
   parent=ReleaseButton
   widget=ButtonTextBlock
   properties {
    text=Release Button
    font {
     bForceMonospaced=false
     bMaterialIsStencil=false
     fontMaterial=null
     fontObject=&Roboto
     letterSpacing=0
     monospacedWidth=1
     size=10.5
     skewAmount=0
     typefaceFontName=Regular
     outlineSettings {
      bApplyOutlineToDropShadows=false
      bMiteredCorners=false
      bSeparateFillAlpha=false
      outlineMaterial=null
      outlineSize=0
      outlineColor=<rgba 0 0 0 1>
     }
    }
   }
   slot {class=ButtonSlot}
  }
 ]
}
