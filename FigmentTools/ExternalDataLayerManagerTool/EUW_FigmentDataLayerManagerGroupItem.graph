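# Editor utility list-entry widget for one group in the Figment External Data Layer Manager tool, showing that group's release items.
# When its list item is set, it casts the object to ExternalDataLayerGroupItem and loops over EDL_ReleaseItems.
# The first item goes into the Release list view and the rest into the ReleaseEventList below it.
# Both nested lists use EUW_FigmentDataLayerManagerItem as their entry widget.
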EditorUtilityWidgetBlueprint /FigmentTools/ExternalDataLayerManagerTool/EUW_FigmentDataLayerManagerGroupItem
graphs [EventGraph {nodes [
   n1 = Event OnListItemObjectSet()
   > n2 = Cast To ExternalDataLayerGroupItem!(execute, Object=@n1.ListItemObject) -> (then, CastFailed, AsExternal Data Layer Group Item:External Data Layer Group Item Object Reference) {is=title K2Node_DynamicCast}
   > n3 = For Each Loop(Array:Array of External Data Layer Release Item Object References=Get EDL_ReleaseItems!(self:External Data Layer Group Item Object Reference=@n2) -> (EDL_ReleaseItems:Array of External Data Layer Release Item Object References)) -> (Array Element:External Data Layer Release Item Object Reference)
   > n4 = Branch(Condition="Equal (Integer)"(A=@n3.Array Index))
   > AddItem!(execute, self:List View Object Reference=Get Release() -> (Release:Editor Utility List View Object Reference), Item:Object Reference=@n3) -> (then)
   AddItem(execute=@n4.else, self=Get ReleaseEventList() -> (ReleaseEventList:Editor Utility List View Object Reference), Item=@n3)
   Event PreConstruct()
   Event Construct()
   Event Tick()
  ]}]
variables []
widgets {
 info {rootWidgetClass=&VerticalBox, widgetCount=3}
 widgets [
  VerticalBox_0 {widgetClassPath=&VerticalBox, widget=VerticalBox_0, properties {bOverride_Cursor=true}}
  Release {
   widgetClassPath=/Script/Blutility.EditorUtilityListView
   bIsVariable=true
   parent=VerticalBox_0
   widget=Release
   properties {
    entryWidgetClass=&/EUW_FigmentDataLayerManagerItem.EUW_FigmentDataLayerManagerItem_C
    numDesignerPreviewEntries=1
    scrollBarPadding=<ltrb 0.1 0.1 0.1 0.1>
    widgetStyle {backgroundBrush {
      bIsDynamicallyLoaded=false
      drawAs=NoDrawType
      imageType=NoImage
      mirroring=NoMirror
      resourceName=None
      resourceObject=null
      tiling=NoTile
      imageSize=<0 0>
      margin=<ltrb 0 0 0 0>
      outlineSettings {
       bUseBrushTransparency=false
       roundingType=HalfHeightRadius
       width=0
       color {colorUseRule=UseColor_Specified, specifiedColor=<rgba 0 0 0 0>}
       cornerRadii=<0 0 0 0>
      }
      tintColor {colorUseRule=UseColor_Specified, specifiedColor=<rgba 0 0 0 1>}
      uVRegion {bIsValid=false, max=<0 0>, min=<0 0>}
     }}
   }
   slot {class=VerticalBoxSlot, properties {padding=<ltrb 4 2 4 2>}}
  }
  ReleaseEventList {
   widgetClassPath=&EditorUtilityListView
   bIsVariable=true
   parent=VerticalBox_0
   widget=ReleaseEventList
   properties {
    entryWidgetClass=&EUW_FigmentDataLayerManagerItem_C
    numDesignerPreviewEntries=8
    scrollBarPadding=<ltrb 0.1 0.1 0.1 0.1>
    widgetStyle {backgroundBrush {
      bIsDynamicallyLoaded=false
      drawAs=NoDrawType
      imageType=NoImage
      mirroring=NoMirror
      resourceName=None
      resourceObject=null
      tiling=NoTile
      imageSize=<0 0>
      margin=<ltrb 0 0 0 0>
      outlineSettings {
       bUseBrushTransparency=false
       roundingType=HalfHeightRadius
       width=0
       color {colorUseRule=UseColor_Specified, specifiedColor=<rgba 0 0 0 0>}
       cornerRadii=<0 0 0 0>
      }
      tintColor {colorUseRule=UseColor_Specified, specifiedColor=<rgba 0 0 0 1>}
      uVRegion {bIsValid=false, max=<0 0>, min=<0 0>}
     }}
   }
   slot {class=VerticalBoxSlot, properties {horizontalAlignment=HAlign_Center, padding=<ltrb 4 2 4 2>}}
  }
 ]
}
