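# Editor utility widget (EDL Release Manager) for viewing External Data Layer releases from a Season EDL data asset.
# PreConstruct creates an ExternalDataLayerHandlerTool, binds the asset picker to it and defaults to S09_EDL_AssetData.
# Construct or asset changes list the handler's release groups, or show its error, and bind release item state changes to the Current State label.
# A checkbox toggles keeping plugins loaded, and SetReleaseAssetDataFromPython swaps the asset by package name.
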EditorUtilityWidgetBlueprint /FigmentTools/ExternalDataLayerManagerTool/EUW_FigmentDataLayerManager
graphs [
 ConstructUpdateView {nodes [
   ConstructUpdateView() {is=title K2Node_FunctionEntry}
   > Clear(TargetArray:Array of External Data Layer Release Item Object References=Get OutReleaseItems() -> (OutReleaseItems:Array of External Data Layer Release Item Object References))
   n1 = Get EDL_HandlerTool() -> (EDL_HandlerTool:External Data Layer Handler Tool Object Reference)
   > n2 = SetEDL_ReleaseGroupItems!(execute, self:External Data Layer Handler Tool Object Reference=@n1) -> (then, ReturnValue:bool)
   > n3 = HandleAssetDataIntegrity!(execute, self, bDataBuildSucceeded:bool=@n2) -> (then, CanSetItemGroups:bool)
   > Branch(Condition=@n3)
   > SetIsChecked!(
    execute
    self:Check Box Object Reference=Get KeepPluginsCheckBox() -> (KeepPluginsCheckBox:Editor Utility Check Box Object Reference)
    InIsChecked:bool=Get!(self:External Data Layer Handler Tool Object Reference=@n1) -> (EDL_CommonPoolData_bKeepAllPluginsLoaded:bool=~)
   ) -> (then)
   > n4 = For Each Loop(Array:Array of External Data Layer Group Item Object References=Get EDL_ReleaseGroupItems!(self:External Data Layer Handler Tool Object Reference=@n1) -> (EDL_ReleaseGroupItems:Array of External Data Layer Group Item Object References)) -> (Array Element:External Data Layer Group Item Object Reference)
   Return Node!(execute=@n4.Completed, OutReleaseItems:Array of External Data Layer Release Item Object References=Get OutReleaseItems())
   > AddItem!(
    execute
    self:List View Object Reference=Get ReleaseEventList() -> (ReleaseEventList:Editor Utility List View Object Reference)
    Item:Object Reference=@n4
   ) -> (then)
   > Append Array!(
    execute
    TargetArray:Array of External Data Layer Release Item Object References=Get OutReleaseItems()
    SourceArray:Array of External Data Layer Release Item Object References=Get EDL_ReleaseItems!(self:External Data Layer Group Item Object Reference=@n4) -> (EDL_ReleaseItems:Array of External Data Layer Release Item Object References)
   ) -> (then)
  ]}
 EventGraph {nodes [
   n1 = "On Check State Changed (KeepPluginsCheckBox)"!() -> (OutputDelegate, then, bIsChecked:bool) {is=title K2Node_ComponentBoundEvent}
   > ChangeKeepPluginsCheckBoxState!(execute, self:External Data Layer Handler Tool Object Reference=Get EDL_HandlerTool(), bIsChecked:bool=@n1.bIsChecked) -> (then)
   n2 = "On Property Changed (SingleProperty_ReleaseAssetData)"!() -> (OutputDelegate, then, PropertyName:Name=None) {is=title K2Node_ComponentBoundEvent}
   n3 = ClearListItems(execute=[@n2.then, @n7.then], self=Get ReleaseEventList())
   n4 = ConstructUpdateView!(execute=[@n3.then, @n8.then], self) -> (then, OutReleaseItems:Array of External Data Layer Release Item Object References) {is=title K2Node_CallFunction}
   > n5 = For Each Loop(Array:Array of External Data Layer Release Item Object References=@n4) -> (Array Element:External Data Layer Release Item Object Reference)
   n6 = OnStateChanged!() -> (OutputDelegate, then, StateName:"Text (by ref)") {is=title K2Node_CustomEvent}
   Bind Event to On State Name Changed(execute=@n5.LoopBody, self:External Data Layer Release Item Object Reference=@n5, Delegate=@n6) {is=title K2Node_AddDelegate}
   n7 = OnReleaseAssetDataChangedFromPython() {is=title K2Node_CustomEvent}
   Event PreConstruct()
   > PreConstructUpdateView()
   n8 = Event Construct()
   Set Text(execute=@n6.then, Text=@n6.StateName, self=Get State() -> (State:Text Object Reference))
  ]}
 HandleAssetDataIntegrity {nodes [
   n1 = HandleAssetDataIntegrity!() -> (then, bDataBuildSucceeded:bool=~) {is=title K2Node_FunctionEntry}
   > n2 = Branch(Condition=@n1)
   n3 = Get ReleaseEventList()
   > SetVisibility(self=@n3, InVisibility=Visible)
   n4 = Get TextError() -> (TextError:Text Block Object Reference)
   > n5 = SetVisibility(self=@n4, InVisibility=Hidden)
   Return Node!(execute=[@n5.then, @n7.then], CanSetItemGroups:bool=@n1)
   SetVisibility(execute=@n2.else, self=@n3, InVisibility=Hidden)
   > SetVisibility(self=@n4, InVisibility=Visible)
   n6 = Get EDL_HandlerTool()
   > n7 = "SetText (Text)"(
    self=@n4
    InText="To Text (String)"(InString=Get ErrorMessage!(self:External Data Layer Handler Tool Object Reference=@n6) -> (ErrorMessage:String))
   )
  ]}
 PreConstructUpdateView {nodes [
   PreConstructUpdateView() {is=title K2Node_FunctionEntry}
   > n1 = Construct Object from Class(Class=/Script/FigmentTools.ExternalDataLayerHandlerTool) -> (ReturnValue:External Data Layer Handler Tool Object Reference)
   > n2 = Set EDL_HandlerTool(EDL_HandlerTool:External Data Layer Handler Tool Object Reference=@n1) -> (Output_Get:External Data Layer Handler Tool Object Reference)
   > SetObject(self=Get SingleProperty_ReleaseAssetData() -> (SingleProperty_ReleaseAssetData:Single Property View Object Reference), NewObject=@n2)
   > Set ReleaseAssetData!(
    execute
    ReleaseAssetData:Season EDL Asset Data Soft Object Reference=&/S09_EDL_AssetData
    self:External Data Layer Handler Tool Object Reference=Get EDL_HandlerTool()
   ) -> (then, Output_Get:Season EDL Asset Data Soft Object Reference)
  ]}
 SetReleaseAssetDataFromPython {nodes [
   n1 = SetReleaseAssetDataFromPython!() -> (then, PackageName:Name) {is=title K2Node_FunctionEntry}
   > n2 = Sequence()
   > n3 = GetAssetsByPackageName!(
    execute
    self:Asset Registry Interface=GetAssetRegistry()
    PackageName:Name=@n1
    bIncludeOnlyOnDiskAssets:bool=false
    bSkipARFilteredAssets:bool=true
   ) -> (then, OutAssetData:Array of Asset Data Structures, ReturnValue:bool)
   > Branch(Condition=@n3.ReturnValue)
   > n4 = Cast To SeasonEDL_AssetData!(execute, Object=GetAsset(InAssetData="Get (a copy)"(Array:Array of Asset Data Structures=@n3) -> (Output:Asset Data Structure))) -> (then, CastFailed, AsSeason EDL Asset Data:Season EDL Asset Data Object Reference) {is=title K2Node_DynamicCast}
   > Set ReleaseAssetData(
    ReleaseAssetData=Make Soft Reference!(Input:Season EDL Asset Data Object Reference=@n4) -> (Output:Season EDL Asset Data Soft Object Reference)
    self=Get EDL_HandlerTool()
   )
   > Set localWasFound(localWasFound:bool=true) -> (Output_Get:bool)
   n5 = Branch(execute=@n2.then_1, Condition=Get localWasFound() -> (localWasFound:bool))
   > OnReleaseAssetDataChangedFromPython() {is=title K2Node_CallFunction}
   Show Message Dialog(
    execute=@n5.else
    Title=Format Text(Format="LOCTEXT(\"Figment Data Layer Manager Error\")")
    Message=Format Text!(Format="LOCTEXT(\"Couldn\\'t find Asset Data: {packageName}\")", packageName:Name=@n1) -> (Result)
    MessageType=Ok
    DefaultValue=Ok
    MessageCategory=Error
   )
  ]}
]
variables [
 EDL_HandlerTool {type=ExternalDataLayerHandlerTool, default=null}
 In List Items {type="array<Object>", default []}
 Plugins in Release 4 Geode {type="array<Object>", default [null,null,null,null]}
 RemovaAllRelesesItem {type=ExternalDataLayerReleaseItem, default=null}
]
widgets {
 info {rootWidgetClass=&StackBox, widgetCount=14}
 widgets [
  StackBox_0 {widgetClassPath=&StackBox, widget=StackBox_0, properties {orientation=Orient_Vertical}}
  TextBlock_67 {
   widgetClassPath=&TextBlock
   parent=StackBox_0
   widget=TextBlock_67
   properties {
    justification=Center
    text=EDL Release Manager
    font {
     bForceMonospaced=false
     bMaterialIsStencil=false
     fontMaterial=null
     fontObject=&Roboto
     letterSpacing=0
     monospacedWidth=1
     size=15
     skewAmount=0
     typefaceFontName=Bold
     outlineSettings {
      bApplyOutlineToDropShadows=false
      bMiteredCorners=false
      bSeparateFillAlpha=false
      outlineMaterial=null
      outlineSize=0
      outlineColor=<rgba 0 0 0 1>
     }
    }
   }
   slot {class=StackBoxSlot, properties {horizontalAlignment=HAlign_Center, padding=<ltrb 0 0 0 10>}}
  }
  SingleProperty_ReleaseAssetData {
   widgetClassPath=&SinglePropertyView
   parent=StackBox_0
   widget=SingleProperty_ReleaseAssetData
   properties {nameOverride=DataAsset, propertyName=ReleaseAssetData}
   slot {class=StackBoxSlot, properties {horizontalAlignment=HAlign_Center, padding=<ltrb 0 0 0 10>}}
  }
  TopBox {
   widgetClassPath=&VerticalBox
   bIsVariable=true
   parent=StackBox_0
   widget=TopBox
   slot {class=StackBoxSlot, properties {size {sizeRule=Fill, value=1}}}
  }
  HorizontalBox {
   widgetClassPath=&HorizontalBox
   parent=TopBox
   widget=HorizontalBox
   properties {toolTipText="If checked, all release plugins will be in the level; if unchecked, only the related ones for the current release will be loaded. A button must be pressed to apply the changes."}
   slot {class=VerticalBoxSlot, properties {horizontalAlignment=HAlign_Center, padding=<ltrb 0 0 0 10>}}
  }
  TextBlock_3 {
   widgetClassPath=&TextBlock
   parent=HorizontalBox
   widget=TextBlock_3
   properties {
    justification=Center
    text="Keep Plugins Loaded: "
    font {
     bForceMonospaced=false
     bMaterialIsStencil=false
     fontMaterial=null
     fontObject=&Roboto
     letterSpacing=0
     monospacedWidth=1
     size=9.75
     skewAmount=0
     typefaceFontName=Bold
     outlineSettings {
      bApplyOutlineToDropShadows=false
      bMiteredCorners=false
      bSeparateFillAlpha=false
      outlineMaterial=null
      outlineSize=0
      outlineColor=<rgba 0 0 0 1>
     }
    }
   }
   slot {class=HorizontalBoxSlot}
  }
  KeepPluginsCheckBox {
   widgetClassPath=&EditorUtilityCheckBox
   bIsVariable=true
   parent=HorizontalBox
   widget=KeepPluginsCheckBox
   properties {
    checkedState=Checked
    widgetStyle {
     checkBoxType=CheckBox
     backgroundHoveredImage {
      bIsDynamicallyLoaded=false
      drawAs=RoundedBox
      imageType=NoImage
      mirroring=NoMirror
      resourceName=None
      resourceObject=null
      tiling=NoTile
      imageSize=<18 18>
      margin=<ltrb 0 0 0 0>
      outlineSettings {
       bUseBrushTransparency=false
       roundingType=FixedRadius
       width=1
       color {colorUseRule=UseColor_ColorTable, specifiedColor=<rgba 0 0 0 0>}
       cornerRadii=<3 3 3 3>
      }
      tintColor {colorUseRule=UseColor_Specified, specifiedColor=<rgba 0.0048 0.0048 0.0048 1>}
      uVRegion {bIsValid=false, max=<0 0>, min=<0 0>}
     }
     backgroundImage {
      bIsDynamicallyLoaded=false
      drawAs=RoundedBox
      imageType=NoImage
      mirroring=NoMirror
      resourceName=None
      resourceObject=null
      tiling=NoTile
      imageSize=<18 18>
      margin=<ltrb 0 0 0 0>
      outlineSettings {
       bUseBrushTransparency=false
       roundingType=FixedRadius
       width=1
       color {colorUseRule=UseColor_ColorTable, specifiedColor=<rgba 0 0 0 0>}
       cornerRadii=<3 3 3 3>
      }
      tintColor {colorUseRule=UseColor_Specified, specifiedColor=<rgba 0.0048 0.0048 0.0048 1>}
      uVRegion {bIsValid=false, max=<0 0>, min=<0 0>}
     }
     backgroundPressedImage {
      bIsDynamicallyLoaded=false
      drawAs=RoundedBox
      imageType=NoImage
      mirroring=NoMirror
      resourceName=None
      resourceObject=null
      tiling=NoTile
      imageSize=<18 18>
      margin=<ltrb 0 0 0 0>
      outlineSettings {
       bUseBrushTransparency=false
       roundingType=FixedRadius
       width=1
       color {colorUseRule=UseColor_ColorTable, specifiedColor=<rgba 0 0 0 0>}
       cornerRadii=<3 3 3 3>
      }
      tintColor {colorUseRule=UseColor_Specified, specifiedColor=<rgba 0.0048 0.0048 0.0048 1>}
      uVRegion {bIsValid=false, max=<0 0>, min=<0 0>}
     }
     borderBackgroundColor {colorUseRule=UseColor_Specified, specifiedColor=<rgba 1 1 1 1>}
     checkedForeground {colorUseRule=UseColor_Specified, specifiedColor=<rgba 1 1 1 1>}
     checkedHoveredForeground {colorUseRule=UseColor_Specified, specifiedColor=<rgba 1 1 1 1>}
     checkedHoveredImage {
      bIsDynamicallyLoaded=false
      drawAs=Image
      imageType=Vector
      mirroring=NoMirror
      resourceName=../../../Engine/Content/Slate/Starship/CoreWidgets/CheckBox/check.svg
      resourceObject=null
      tiling=NoTile
      imageSize=<16 16>
      margin=<ltrb 0 0 0 0>
      outlineSettings {
       bUseBrushTransparency=false
       roundingType=HalfHeightRadius
       width=0
       color {colorUseRule=UseColor_Specified, specifiedColor=<rgba 0 0 0 0>}
       cornerRadii=<0 0 0 0>
      }
      tintColor {colorUseRule=UseColor_Specified, specifiedColor=<rgba 0.0044 0.2384 1 1>}
      uVRegion {bIsValid=false, max=<0 0>, min=<0 0>}
     }
     checkedImage {
      bIsDynamicallyLoaded=false
      drawAs=Image
      imageType=Vector
      mirroring=NoMirror
      resourceName=../../../Engine/Content/Slate/Starship/CoreWidgets/CheckBox/check.svg
      resourceObject=null
      tiling=NoTile
      imageSize=<16 16>
      margin=<ltrb 0 0 0 0>
      outlineSettings {
       bUseBrushTransparency=false
       roundingType=HalfHeightRadius
       width=0
       color {colorUseRule=UseColor_Specified, specifiedColor=<rgba 0 0 0 0>}
       cornerRadii=<0 0 0 0>
      }
      tintColor {colorUseRule=UseColor_Specified, specifiedColor=<rgba 0 0.162 0.7454 1>}
      uVRegion {bIsValid=false, max=<0 0>, min=<0 0>}
     }
     checkedPressedForeground {colorUseRule=UseColor_Specified, specifiedColor=<rgba 1 1 1 1>}
     checkedPressedImage {
      bIsDynamicallyLoaded=false
      drawAs=Image
      imageType=Vector
      mirroring=NoMirror
      resourceName=../../../Engine/Content/Slate/Starship/CoreWidgets/CheckBox/check.svg
      resourceObject=null
      tiling=NoTile
      imageSize=<16 16>
      margin=<ltrb 0 0 0 0>
      outlineSettings {
       bUseBrushTransparency=false
       roundingType=HalfHeightRadius
       width=0
       color {colorUseRule=UseColor_Specified, specifiedColor=<rgba 0 0 0 0>}
       cornerRadii=<0 0 0 0>
      }
      tintColor {colorUseRule=UseColor_Specified, specifiedColor=<rgba 0 0.0802 0.3515 1>}
      uVRegion {bIsValid=false, max=<0 0>, min=<0 0>}
     }
     checkedSlateSound {resourceObject=null}
     foregroundColor {colorUseRule=UseColor_Specified, specifiedColor=<rgba 1 1 1 1>}
     hoveredForeground {colorUseRule=UseColor_Specified, specifiedColor=<rgba 1 1 1 1>}
     hoveredSlateSound {resourceObject=null}
     padding=<ltrb 0 0 0 0>
     pressedForeground {colorUseRule=UseColor_Specified, specifiedColor=<rgba 1 1 1 1>}
     uncheckedHoveredImage {
      bIsDynamicallyLoaded=false
      drawAs=NoDrawType
      imageType=NoImage
      mirroring=NoMirror
      resourceName=None
      resourceObject=null
      tiling=NoTile
      imageSize=<0 0>
      margin=<ltrb 0 0 0 0>
      outlineSettings {
       bUseBrushTransparency=false
       roundingType=HalfHeightRadius
       width=0
       color {colorUseRule=UseColor_Specified, specifiedColor=<rgba 0 0 0 0>}
       cornerRadii=<0 0 0 0>
      }
      tintColor {colorUseRule=UseColor_Specified, specifiedColor=<rgba 1 1 1 1>}
      uVRegion {bIsValid=false, max=<0 0>, min=<0 0>}
     }
     uncheckedImage {
      bIsDynamicallyLoaded=false
      drawAs=NoDrawType
      imageType=NoImage
      mirroring=NoMirror
      resourceName=None
      resourceObject=null
      tiling=NoTile
      imageSize=<0 0>
      margin=<ltrb 0 0 0 0>
      outlineSettings {
       bUseBrushTransparency=false
       roundingType=HalfHeightRadius
       width=0
       color {colorUseRule=UseColor_Specified, specifiedColor=<rgba 0 0 0 0>}
       cornerRadii=<0 0 0 0>
      }
      tintColor {colorUseRule=UseColor_Specified, specifiedColor=<rgba 1 1 1 1>}
      uVRegion {bIsValid=false, max=<0 0>, min=<0 0>}
     }
     uncheckedPressedImage {
      bIsDynamicallyLoaded=false
      drawAs=NoDrawType
      imageType=NoImage
      mirroring=NoMirror
      resourceName=None
      resourceObject=null
      tiling=NoTile
      imageSize=<0 0>
      margin=<ltrb 0 0 0 0>
      outlineSettings {
       bUseBrushTransparency=false
       roundingType=HalfHeightRadius
       width=0
       color {colorUseRule=UseColor_Specified, specifiedColor=<rgba 0 0 0 0>}
       cornerRadii=<0 0 0 0>
      }
      tintColor {colorUseRule=UseColor_Specified, specifiedColor=<rgba 1 1 1 1>}
      uVRegion {bIsValid=false, max=<0 0>, min=<0 0>}
     }
     uncheckedSlateSound {resourceObject=null}
     undeterminedForeground {colorUseRule=UseColor_Specified, specifiedColor=<rgba 1 1 1 1>}
     undeterminedHoveredImage {
      bIsDynamicallyLoaded=false
      drawAs=Image
      imageType=Vector
      mirroring=NoMirror
      resourceName=../../../Engine/Content/Slate/Starship/CoreWidgets/CheckBox/indeterminate.svg
      resourceObject=null
      tiling=NoTile
      imageSize=<16 16>
      margin=<ltrb 0 0 0 0>
      outlineSettings {
       bUseBrushTransparency=false
       roundingType=HalfHeightRadius
       width=0
       color {colorUseRule=UseColor_Specified, specifiedColor=<rgba 0 0 0 0>}
       cornerRadii=<0 0 0 0>
      }
      tintColor {colorUseRule=UseColor_Specified, specifiedColor=<rgba 0.0044 0.2384 1 1>}
      uVRegion {bIsValid=false, max=<0 0>, min=<0 0>}
     }
     undeterminedImage {
      bIsDynamicallyLoaded=false
      drawAs=Image
      imageType=Vector
      mirroring=NoMirror
      resourceName=../../../Engine/Content/Slate/Starship/CoreWidgets/CheckBox/indeterminate.svg
      resourceObject=null
      tiling=NoTile
      imageSize=<16 16>
      margin=<ltrb 0 0 0 0>
      outlineSettings {
       bUseBrushTransparency=false
       roundingType=HalfHeightRadius
       width=0
       color {colorUseRule=UseColor_Specified, specifiedColor=<rgba 0 0 0 0>}
       cornerRadii=<0 0 0 0>
      }
      tintColor {colorUseRule=UseColor_Specified, specifiedColor=<rgba 0 0.162 0.7454 1>}
      uVRegion {bIsValid=false, max=<0 0>, min=<0 0>}
     }
     undeterminedPressedImage {
      bIsDynamicallyLoaded=false
      drawAs=Image
      imageType=Vector
      mirroring=NoMirror
      resourceName=../../../Engine/Content/Slate/Starship/CoreWidgets/CheckBox/indeterminate.svg
      resourceObject=null
      tiling=NoTile
      imageSize=<16 16>
      margin=<ltrb 0 0 0 0>
      outlineSettings {
       bUseBrushTransparency=false
       roundingType=HalfHeightRadius
       width=0
       color {colorUseRule=UseColor_Specified, specifiedColor=<rgba 0 0 0 0>}
       cornerRadii=<0 0 0 0>
      }
      tintColor {colorUseRule=UseColor_Specified, specifiedColor=<rgba 0 0.0802 0.3515 1>}
      uVRegion {bIsValid=false, max=<0 0>, min=<0 0>}
     }
    }
   }
   slot {class=HorizontalBoxSlot, properties {size {sizeRule=Fill, value=1}}}
  }
  HorizontalBox_0 {
   widgetClassPath=&HorizontalBox
   parent=TopBox
   widget=HorizontalBox_0
   properties {toolTipText=The currently loaded State.}
   slot {class=VerticalBoxSlot, properties {padding=<ltrb 0 0 0 10>}}
  }
  Calendar {
   widgetClassPath=&TextBlock
   parent=HorizontalBox_0
   widget=Calendar
   properties {
    justification=Right
    text=Current State:
    font {
     bForceMonospaced=false
     bMaterialIsStencil=false
     fontMaterial=null
     fontObject=&Roboto
     letterSpacing=0
     monospacedWidth=1
     size=10.5
     skewAmount=0
     typefaceFontName=Regular
     outlineSettings {
      bApplyOutlineToDropShadows=false
      bMiteredCorners=false
      bSeparateFillAlpha=false
      outlineMaterial=null
      outlineSize=0
      outlineColor=<rgba 0 0 0 1>
     }
    }
   }
   slot {class=HorizontalBoxSlot, properties {horizontalAlignment=HAlign_Right, size {sizeRule=Fill, value=1}}}
  }
  State {
   widgetClassPath=&TextBlock
   bIsVariable=true
   parent=HorizontalBox_0
   widget=State
   properties {
    text=None
    font {
     bForceMonospaced=false
     bMaterialIsStencil=false
     fontMaterial=null
     fontObject=&Roboto
     letterSpacing=0
     monospacedWidth=1
     size=10.5
     skewAmount=0
     typefaceFontName=Regular
     outlineSettings {
      bApplyOutlineToDropShadows=false
      bMiteredCorners=false
      bSeparateFillAlpha=false
      outlineMaterial=null
      outlineSize=0
      outlineColor=<rgba 0 0 0 1>
     }
    }
   }
   slot {class=HorizontalBoxSlot, properties {horizontalAlignment=HAlign_Left, size {sizeRule=Fill, value=1}}}
  }
  TopBox_4 {
   widgetClassPath=&VerticalBox
   bIsVariable=true
   parent=TopBox
   widget=TopBox_4
   slot {class=VerticalBoxSlot, properties {size {sizeRule=Fill, value=1}}}
  }
  VerticalBox_ContainersToMap {
   widgetClassPath=&EditorUtilityScrollBox
   bIsVariable=true
   parent=TopBox_4
   widget=VerticalBox_ContainersToMap
   properties {alwaysShowScrollbar=true, alwaysShowScrollbarTrack=true, consumeMouseWheel=Always}
   slot {class=VerticalBoxSlot, properties {size {sizeRule=Fill, value=1}}}
  }
  ReleaseEventList {
   widgetClassPath=/Script/Blutility.EditorUtilityListView
   bIsVariable=true
   parent=VerticalBox_ContainersToMap
   widget=ReleaseEventList
   properties {
    entryWidgetClass=&/EUW_FigmentDataLayerManagerGroupItem.EUW_FigmentDataLayerManagerGroupItem_C
    numDesignerPreviewEntries=8
    scrollBarPadding=<ltrb 0.1 0.1 0.1 0.1>
    widgetStyle {backgroundBrush {
      bIsDynamicallyLoaded=false
      drawAs=NoDrawType
      imageType=NoImage
      mirroring=NoMirror
      resourceName=None
      resourceObject=null
      tiling=NoTile
      imageSize=<0 0>
      margin=<ltrb 0 0 0 0>
      outlineSettings {
       bUseBrushTransparency=false
       roundingType=HalfHeightRadius
       width=0
       color {colorUseRule=UseColor_Specified, specifiedColor=<rgba 0 0 0 0>}
       cornerRadii=<0 0 0 0>
      }
      tintColor {colorUseRule=UseColor_Specified, specifiedColor=<rgba 0 0 0 1>}
      uVRegion {bIsValid=false, max=<0 0>, min=<0 0>}
     }}
   }
   slot {class=ScrollBoxSlot, properties {padding=<ltrb 4 2 4 2>}}
  }
  TextError {
   widgetClassPath=&UEFN_TextBlock
   bIsVariable=true
   parent=VerticalBox_ContainersToMap
   widget=TextError
   properties {
    text=Error Message
    font {
     bForceMonospaced=false
     bMaterialIsStencil=false
     fontMaterial=null
     fontObject=&BurbankSmall
     letterSpacing=0
     monospacedWidth=1
     size=10.5
     skewAmount=0
     typefaceFontName=Black
     outlineSettings {
      bApplyOutlineToDropShadows=false
      bMiteredCorners=false
      bSeparateFillAlpha=false
      outlineMaterial=null
      outlineSize=0
      outlineColor=<rgba 0 0 0 1>
     }
    }
   }
   slot {class=ScrollBoxSlot}
  }
 ]
}
